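# Physics asset for the M_MED_Foam backpack cosmetic, simulating its foam finger attachment.
# It has two capsule bodies, a small one on dyn_finger_master_jnt and a larger tilted one on dyn_finger_jnt.
# A single constraint links them with tight swing limits and a looser twist limit, so the finger moves only slightly relative to its master joint.
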PhysicsAsset /BRCosmetics/Accessories/FORT_Backpacks/Backpack_M_MED_Foam/Meshes/M_MED_Foam_Pack_Physics
bodies {
 dyn_finger_jnt [{shapeType=Capsule, length=7, radius=3, shapeName=dyn_finger_jnt_capsule, center=<0.8704 0 6.1703>, rotation=<rot -10.0001 0 0>}]
 dyn_finger_master_jnt [{shapeType=Capsule, length=0.505, radius=0.505, shapeName=dyn_finger_master_jnt_capsule, rotation=<rot 0 -90 0>}]
}
constraints [{bone1Name=dyn_finger_jnt, bone2Name=dyn_finger_master_jnt, swing1LimitDegrees=5, swing2LimitDegrees=5, twistLimitDegrees=25}]
